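# Niagara effect for the SilentNovel emote's shake, a burst of small, short-lived mesh sprinkles spawned around two fork sockets on the skeletal mesh.
# Particles take weighted palette colors with white rarest, fly out under gravity and drag, tumble, and scale in and out over their lifetime.
# Toon shader outline width and particle brightness follow time of day, and positions use a world/local blend.
# The collision module is disabled.
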NiagaraSystem /BRCosmetics/Animation/Game/MainPlayer/Emotes/SilentNovel/FX/NS_Emote_SilentNovel_Shake
dependencies {
 usedDataInterfaces [
  &NiagaraDataInterfaceSkeletalMesh
  /Script/Niagara.NiagaraDataInterfaceArrayDistributionInt
  &NiagaraDataInterfaceArrayColor
  &NiagaraDataInterfaceMeshRendererInfo
  &NiagaraDataInterfaceCurve
 ]
 usedDynamicInputs []
 usedModules [&SystemState, &BlendSpaceHelper]
}
emitters {NE_Sprinkles001 {
  emitterSpawnScript [
   /Game/Effects/Niagara/MasterMaterials/ToonShader/Template/NM_OutlineParameters(OutlineWidth:NiagaraFloat=&Multiply_Float(A=0.35, B=&FNTimeOfDayScalar(Night Scale=0.6, Morning Scale=1, Evening Scale=1)))
   /Game/Effects/Niagara/MasterMaterials/ToonShader/Template/NM_ToonShaderParameters(HighLight_Size=0.1, HighLight_Softness=0.05, Lighting_Hardness=1, Metallic=0.5, Metallic_LightingPower=0)
  ]
  emitterUpdateScript [
   &EmitterState(Life Cycle Mode=::NewEnumerator1, Loop Behavior=::NewEnumerator1, Loop Duration=1, UseLoopDelay=false)
   &SpawnBurst_Instantaneous(Spawn Count=50, Use Spawn Probability=false, UseLoopCountLimit=false)
  ]
  particleSpawnScript [
   &SkeletalMeshLocation(
    Bone / Socket Index=&RandomRangeInt(Minimum=0, Maximum=1)
    MaskSpawnGroup=false
    Mesh Sampling Type=::NewEnumerator5
    Position Sampling=::NewEnumerator2
    Skeletal Mesh=DataInterface() {propertyValues {
      DefaultMesh=null
      ExcludeBoneName=root
      OptionalComponentClass=null
      PreviewMesh=null
      SkinningMode=SkinOnTheFly
      SoftSourceActor=null
      SourceMode=Default
      UvSetIndex=0
      WholeMeshLOD=-1
      bExcludeBone=false
      bReadDeformedGeometry=true
      bRequireCurrentFrameData=true
      ComponentTags []
      FilteredBones []
      FilteredSockets [JNT_A_ForkSocket, JNT_B_ForkSocket]
      MeshUserParameter {parameter {name=None, typeDefHandle {registeredTypeIndex=1}, varData []}}
      SamplingRegions []
     }}
    Socket Sampling Mode=::NewEnumerator5
   )
   &InitializeParticle(
    Color=&SelectLinearColorFromArray(
     Array Sampling Mode=::NewEnumerator1
     Direct Array Index=/Niagara/DynamicInputs/Arrays/SelectIntFromWeightedDistributionArray(
      Array Sampling Mode:ENiagara_LifetimeMode=::NewEnumerator1
      Direct Array Index:NiagaraInt32=?
      Direct Array Mode:ENiagara_WrapClamp=::NewEnumerator0
      Evaluation Type:ENiagaraRandomnessEvaluation=::NewEnumerator0
      Recalculate Random Each Loop:NiagaraBool=false
      Weighted Distribution Int Array:NiagaraDataInterfaceArrayDistributionInt=DataInterface() {propertyValues {ArrayData [{value=0, weight=0.667}, {value=1, weight=0.667}, {value=2, weight=0.667}, {value=3, weight=0.667}, {value=4, weight=0.2}]}}
      Randomness Mode:ENiagaraRandomnessMode=::NewEnumerator0
      Random Seed:NiagaraInt32=0
      Fixed Random Seed:NiagaraBool=false
     )
     Color Selection Array=DataInterface() {propertyValues {ColorData [<rgba 0.8177 0.0607 0.1942 1>, <rgba 0 0.5078 1 1>, <rgba 0 0.3438 0.0933 1>, <rgba 1 0.7786 0 1>, <rgba 1 1 1 1>]}}
     Random Seed=?
    )
    Lifetime=&RandomRangeFloat(Minimum=0.3, Maximum=0.7)
    Mass Max=1.1
    Mass Min=0.9
    Mass Mode=::NewEnumerator3
    Mesh Renderer Array Visibility Mode=::NewEnumerator1
    Mesh Renderer Info=DataInterface() {propertyValues {MeshRenderer=&self:NE_Sprinkles001_1.NiagaraMeshRendererProperties_7}}
    Mesh Scale Mode=::NewEnumerator3
    Mesh Uniform Scale=&RandomRangeFloat(Minimum=0.05, Maximum=0.15)
    Position Mode=::NewEnumerator1
    Position Offset:NiagaraPosition=$Particles.SkeletalMeshLocation.SampledPosition
    UsePositionOffset=true
    Write Material Random=true
   )
   &ShapeLocation(MaskSpawnGroup=false, Override Seed=false, Sphere Radius=10, UseOwnerScale=true)
   &AddVelocity(Velocity=&RandomRangeVector(Minimum=<50 0 300>, Maximum=<300 1 100>)) {module=AddVelocity001}
   &AddVelocity(
    Module.Use Falloff Distance=false
    Override Seed=false
    Velocity Mode=::NewEnumerator1
    Velocity Origin=$Particles.SkeletalMeshLocation.SampledPosition
    Velocity Speed=&RandomRangeFloat(Minimum=50, Maximum=100)
   )
   &InitialMeshOrientation(Use Rotation Vector=false)
  ]
  particleUpdateScript [
   &ScaleColor(
    Scale RGB=&Multiply_VectorByFloat(
     Vector=<1 1 1>
     Float=&FNTimeOfDayScalar(Night Scale=0.35, Morning Scale=0.9, Day Scale=0.9, Evening Scale=0.9)
    )
    ScaleA=true
    ScaleRGB=true
   )
   &ParticleState()
   &GravityForce()
   &UpdateMeshOrientation(Rotation Rate=2, Rotation Vector=&RandomRangeVector(Minimum=<-0.5 -0.5 -0.5>, Maximum=<0.5 0.5 0.5>))
   &Drag(Use Linear Drag=true, Use Rotational Drag=true)
   &ScaleMeshSize(Scale Factor=&VectorFromFloat(Value=&FloatFromCurve(FloatCurve=DataInterface() {propertyValues {Curve {keys [
          {arriveTangent=-0.0278, interpMode=RCIM_Cubic, time=0.0008, value=0}
          {arriveTangent=0.8536, time=0.1, value=1}
          {arriveTangent=-1.1111, interpMode=RCIM_Cubic, leaveTangent=-1.1111, time=0.7, value=1}
          <at 1 0 RCIM_Cubic>
         ]}}})))
   &Collision(
    Age Colliding Particles=false
    Control Roll On Collision=true
    KillOnCollision=false
    Module.Control Rest State Via Surface Normal=false
    Module.EnableMaxCollisionCount=false
    Module.Randomize Collision Normal Vector=false
    Radius Calculation Type=::NewEnumerator3
    Restitution=&RandomRangeFloat(Minimum=0.3, Maximum=0.4)
    ScaleRateOfRotation=true
   ) {enabled=false}
   &MeshRotationForce(Rotation=&Multiply_VectorByFloat(
     Vector=$Particles.RotationalVelocity
     Float=&MakeCustomFloatFromBool(Bool=$Particles.HasCollided, True Float=1, False Float=0)
    ))
   &SolveForcesAndVelocity(Clamp Velocity=false, Limit Acceleration=false)
   &BlendSpace("Position Blend (World <-> Local)"=0.8)
  ]
  rendererClasses [&NiagaraMeshRendererProperties]
  renderers [
   {rendererClass=&NiagaraMeshRendererProperties, rendererIndex=0}
   {rendererClass=&NiagaraMeshRendererProperties, rendererIndex=1}
  ]
 }}
summary {
 systemName=NS_Emote_SilentNovel_Shake
 emitters [{emitterName=NE_Sprinkles001, rendererClasses [&NiagaraMeshRendererProperties]}]
 userVariables [User.NormalizedNotifyProgress {type=NiagaraFloat, defaultValue=0, description=""}]
}
